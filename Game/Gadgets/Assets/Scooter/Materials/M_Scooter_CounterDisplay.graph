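# Scooter counter screen material, a lit surface showing a tinted image plus a two-digit number set through custom primitive data.
# Fmod math splits the number into tens and units, picking each half's digit from a vertical flipbook atlas with kerning masks.
# Four offset texture samples outline the image, while panning noise gated by summed sines adds intermittent UV glitches.
# Scrolling scanlines modulate emissive, ScreenOffScalar dims base color and emissive, and MF_CreativeMoveTool wraps the attributes.
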Material /Game/Gadgets/Assets/Scooter/Materials/M_Scooter_CounterDisplay
graph {
 properties {
  bUseMaterialAttributes=true
  bUsedWithMeshDeformer=true
  bUsedWithSkeletalMesh=true
  pixelDepthOffsetMode=PDOM_Legacy
  usedShadingModels=Default Lit
 }
 nodes [
  n1 = TextureCoordinate() {uTiling=0}
  n2 = AppendVector(
   A=TextureSample(UVs=Panner(Coordinate=@n1) {bFractionalPart=true, speedX=0.025, speedY=0.3844}) {samplerType=SAMPLERTYPE_Normal, texture=/Game/Global/Textures/Default/Noise/64x64Noise}.R
   B=Constant()
  )
  n3 = Time()
  n4 = Sine(None=@n3) {period=0.355}
  n5 = Sine(None=@n3) {period=0.8755}
  n6 = Power(Base=Saturate(None=Add(A=@n4, B=@n5))) {constExponent=5}
  n7 = Multiply(A=ScalarParameter() {defaultValue=0.01, parameterName=DistortScale}, B=ComponentMask(None=VertexInterpolator(VS=@n6).PS) {r=true})
  n8 = MaterialFunctionCall(UVs=Add(A=TextureCoordinate(), B=Multiply(A=@n2, B=@n7))) {materialFunction=&ScaleUVsByCenter}
  n9 = MaterialFunctionCall(Float2=@n8.UVs) {materialFunction=&BreakOutFloat2Components}
  n10 = ScalarParameter() {defaultValue=1, group=Counter, parameterName=ImageScaleU}
  n11 = ScalarParameter() {defaultValue=1, group=Counter, parameterName=ImageScaleV}
  n12 = MaterialFunctionCall(
   X=Add(A=Multiply(A=@n9.R, B=@n10), B=ScalarParameter() {group=Counter, parameterName=ImageShiftU})
   Y=Add(A=Multiply(A=@n9.G, B=@n11), B=ScalarParameter() {group=Counter, parameterName=ImageShiftV})
  ) {materialFunction=&MakeFloat2}
  n13 = Power(
   Base=TextureSample(UVs=NamedRerouteDeclaration(None=@n12.Result) {name=UVLocation, nodeColor=<rgba 0 0.0625 1 1>}) {texture=/Game/Creative/Devices/Class_Designer/Textures/T_Class_Designer_Screen}.R
   Exp=ScalarParameter() {defaultValue=1, group=Counter, parameterName=ImageContrast}
  )
  n14 = ScalarParameter() {defaultValue=0.5, group=Counter, parameterName=EdgeBrightnessScalar, sliderMax=1}
  n15 = VectorParameter() {group=Counter, parameterName=CounterBaseColorTint, defaultValue=<rgba 0.068 0.4558 0.5538 0>}
  n16 = NamedRerouteUsage()
  n17 = Ceil(None=Subtract(A=TextureSample(UVs=@n16) {texture=&T_Class_Designer_Screen}.R) {constB=0.2})
  n18 = ScalarParameter() {defaultValue=0.02, group=Counter, parameterName=ImageEdgeThickness, sliderMax=1}
  n19 = Multiply(A=Constant2Vector() {g=1}, B=@n18)
  n20 = Subtract(A=TextureSample(UVs=Add(A=@n19, B=@n16)) {texture=&T_Class_Designer_Screen}.R) {constB=0.3}
  n21 = Multiply(A=Constant2Vector() {g=-1}, B=@n18)
  n22 = Subtract(A=TextureSample(UVs=Add(A=@n21, B=@n16)) {texture=&T_Class_Designer_Screen}.R) {constB=0.3}
  n23 = Add(A=Subtract(A=@n17, B=Ceil(None=@n20)), B=Subtract(A=@n17, B=Ceil(None=@n22)))
  n24 = Multiply(A=Constant2Vector() {r=1}, B=@n18)
  n25 = Subtract(A=TextureSample(UVs=Add(A=@n24, B=@n16)) {texture=&T_Class_Designer_Screen}.R) {constB=0.3}
  n26 = Multiply(A=Constant2Vector() {r=-1}, B=@n18)
  n27 = Subtract(A=TextureSample(UVs=Add(A=@n26, B=@n16)) {texture=&T_Class_Designer_Screen}.R) {constB=0.3}
  n28 = Add(A=Subtract(A=@n17, B=Ceil(None=@n25)), B=Subtract(A=@n17, B=Ceil(None=@n27)))
  n29 = Multiply(A=Power(Base=Add(A=@n23, B=@n28)) {constExponent=3}) {constB=3}
  n30 = Multiply(
   A=@n13
   B=LinearInterpolate(
    A=VectorParameter() {group=Counter, parameterName=ImageBaseColorTint, defaultValue=<rgba 0.068 0.4558 0.5538 0>}.RGB
    B=Multiply(A=@n14, B=@n15.RGB)
    Alpha=@n29
   )
  )
  n31 = Add(
   A=MaterialFunctionCall(
    UVs=@n8.UVs
    Rotation Center=Constant2Vector() {g=0.5, r=0.5}
    "Rotation Angle (0-1)"=ScalarParameter() {group=Counter, parameterName=Counter_Rotation}
   ) {materialFunction=&CustomRotator}.Rotated Values
   B=AppendVector(A=ScalarParameter() {group=Counter, parameterName=CounterPosU}, B=ScalarParameter() {group=Counter, parameterName=CounterPosV})
  )
  n32 = MaterialFunctionCall(
   UVs=@n31
   Texture Scale=MaterialFunctionCall(
    X=ScalarParameter() {defaultValue=1, group=Counter, parameterName=CounterScaleU}
    Y=ScalarParameter() {defaultValue=1, group=Counter, parameterName=CounterScaleV}
   ) {materialFunction=&MakeFloat2}.Result
  ) {materialFunction=&ScaleUVsByCenter}
  n33 = Frac(None=@n32.UVs)
  n34 = Constant2Vector() {g=1, r=2}
  n35 = Add(A=Frac(None=Multiply(A=@n33, B=@n34))) {constB=-0.25}
  n36 = ComponentMask(None=@n35) {r=true}
  n37 = ScalarParameter() {defaultValue=0.02, group=Counter - Kerning, parameterName=KerningEdgeMaskSize}
  n38 = Subtract(A=ConstantBiasScale(None=Sine(None=@n36)), B=@n37) {constB=0.1}
  n39 = ComponentMask(None=@n35) {g=true}
  n40 = Subtract(A=ConstantBiasScale(None=Sine(None=@n39)), B=@n37) {constB=0.1}
  n41 = Multiply(A=Multiply(A=Ceil(None=@n38), B=Ceil(None=@n40))) {constB=5}
  n42 = MaterialFunctionCall(Float2=Multiply(A=@n32.UVs, B=@n34)) {materialFunction=&BreakOutFloat2Components}
  n43 = Multiply(A=Frac(None=@n42.G)) {constB=0.1}
  n44 = ScalarParameter() {bUseCustomPrimitiveData=true, defaultValue=55, desc=input number, group=Counter, parameterName=Counter_InputNumber}
  n45 = Constant() {r=10}
  n46 = Fmod(A=@n44, B=@n45)
  n47 = Subtract(A=@n44, B=@n46)
  n48 = ComponentMask(None=@n31) {r=true}
  n49 = Divide(
   A=LinearInterpolate(A=Divide(A=@n47, B=@n45) {constB=10}, B=@n46, Alpha=Round(None=@n48)) {desc=left size is tens. right side is units}
   B=ScalarParameter() {defaultValue=10, desc=Number of Total Digits in Texture, group=Counter, parameterName=numColumns}
  ) {desc="multiply by how many digits in the texture.\r\nThis gives us an offset."}
  n50 = TextureSampleParameter2D(UVs=MaterialFunctionCall(X=@n42.R, Y=Add(A=@n43, B=@n49)) {materialFunction=&MakeFloat2}.Result) {
   group=Counter
   parameterName=Counter_AtlasTexture
   texture=/Game/Accessories/FORT_Backpacks/Backpack_F_MED_Cotton_Candy/Textures/T_NumberFlipBook
  }
  n51 = Multiply(A=@n32.0-1 mask, B=Multiply(A=Clamp(None=@n41), B=@n50.R))
  n52 = Multiply(A=@n51, B=@n15.RGB)
  n53 = ScalarParameter() {defaultValue=1, group=Counter, parameterName=ScreenOffScalar, sliderMax=1}
  n54 = VectorParameter() {group=Counter, parameterName=CounterEmissiveTint, defaultValue=<rgba 0.1357 1.2 0.6057 0>}
  n55 = Multiply(
   A=@n13
   B=LinearInterpolate(
    A=VectorParameter() {group=Counter, parameterName=ImageEmissiveTint, defaultValue=<rgba 0.1357 1.2 0.6057 0>}.RGB
    B=Multiply(A=@n14, B=@n54.RGB)
    Alpha=@n29
   )
  )
  n56 = ScalarParameter() {defaultValue=3, group=Counter, parameterName=EmissiveExponent}
  n57 = ComponentMask(None=Panner(Coordinate=TextureCoordinate()) {speedY=0.02}) {g=true}
  n58 = ScalarParameter() {defaultValue=25, parameterName=ScanlineRepeat}
  n59 = SphereMask(
   A=Frac(None=Multiply(A=@n57, B=@n58))
   B=Constant() {r=0.5}
   Radius=ScalarParameter() {defaultValue=0.4, parameterName=ScanlineRadius}
   Hardness=ScalarParameter() {defaultValue=0.2, parameterName=ScanlineHarndess}
  )
  n60 = Multiply(A=Multiply(A=Power(Base=@n55, Exp=@n56), B=ScalarParameter() {defaultValue=4, group=Counter, parameterName=EmissiveScale}), B=@n59)
  n61 = Multiply(A=@n51, B=@n54.RGB)
  n62 = ScalarParameter() {defaultValue=3, group=Counter, parameterName=EmissiveExponent}
  n63 = Multiply(A=Multiply(A=Power(Base=@n61, Exp=@n62), B=ScalarParameter() {defaultValue=4, group=Counter, parameterName=EmissiveScale}), B=@n59)
  n64 = SetMaterialAttributes!(
   MaterialAttributes
   Base Color=Multiply(A=Add(A=@n30, B=@n52), B=@n53)
   Metallic=ScalarParameter() {parameterName=Metallic}
   Specular=ScalarParameter() {defaultValue=0.8, parameterName=Specular}
   Roughness=ScalarParameter() {defaultValue=0.1, parameterName=Roughness}
   Emissive Color=Multiply(A=Add(A=@n60, B=@n63), B=@n53)
  ) {attributeSetTypes [
    #1
    #2
    #3
    #4
    #5
   ]}
  MaterialFunctionCall(MA In=@n64) {materialFunction=&MF_CreativeMoveTool}
  TextureCoordinate() {coordinateIndex=1}
 ]
}
